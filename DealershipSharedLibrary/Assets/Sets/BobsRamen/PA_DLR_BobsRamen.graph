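# Physics asset for the Bob's Ramen dealership set piece, giving its skeletal mesh one capsule body per listed bone.
# Constraints branch from a tiny Root body into a large Ramen body and a large bob character body.
# Ramen leads a serial chain of cover segments, while bob carries the eyebrows, lower mouth and a two-link hair chain.
# Every constraint shares identical swing and twist limit angles, with no per-joint tuning of those angles.
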PhysicsAsset /DealershipSharedLibrary/Assets/Sets/BobsRamen/PA_DLR_BobsRamen
bodies {
 Cover_12 [{shapeType=Capsule, length=83.0839, radius=40.8126, center=<-27.0555 -0.4893 0.9329>, rotation=<rot 0 2.2566 -89.9999>}]
 Ramen [{shapeType=Capsule, length=216.7978, radius=216.7906, center=<-0.1311 -0.7812 100.8831>, rotation=<rot 0.4665 -89.9999 -90>}]
 Root [{shapeType=Capsule, length=0.505, radius=0.505, rotation=<rot 0 -90 0>}]
 bob [{shapeType=Capsule, length=321.2414, radius=188.7141, center=<18.6246 -0.0993 318.4981>, rotation=<rot -0.0483 -90 -1.9517>}]
 cover_01 [{shapeType=Capsule, length=180.5967, radius=54.6315, center=<53.129 -11.6117 0.9329>, rotation=<rot 0 5.26 -90>}]
 cover_03 [{shapeType=Capsule, length=178.3855, radius=20.3641, center=<30.5209 9.728 0.9331>, rotation=<rot 0 -179.2677 -89.9995>}]
 cover_04 [{shapeType=Capsule, length=200.7379, radius=38.9093, center=<31.8766 0.017 0.9329>, rotation=<rot 0 0.0472 -90>}]
 cover_05 [{shapeType=Capsule, length=214.6682, radius=31.7015, center=<12.1844 -4.3099 0.9329>, rotation=<rot 0 2.7168 -89.9999>}]
 cover_06 [{shapeType=Capsule, length=209.7621, radius=24.3632, center=<13.1518 6.9701 0.9331>, rotation=<rot 0 -176.229 -89.9997>}]
 cover_07 [{shapeType=Capsule, length=216.7657, radius=31.294, center=<11.7679 2.0011 0.9329>, rotation=<rot 0 -177.4338 -89.9999>}]
 cover_08 [{shapeType=Capsule, length=200.7361, radius=26.2497, center=<29.5237 0.059 0.9329>, rotation=<rot 0 0.249 -90>}]
 cover_09 [{shapeType=Capsule, length=172.3486, radius=45.1572, center=<37.1662 -0.2309 0.9329>, rotation=<rot 0 178.9496 -90>}]
 eyebrow_l [{shapeType=Capsule, length=21.9391, radius=12.3636, rotation=<rot 0 -179.9387 -69.2399>}]
 eyebrow_r [{shapeType=Capsule, length=21.939, radius=12.3636, rotation=<rot 0 -0.0614 -69.2398>}]
 hair_01 [{shapeType=Capsule, length=38.0234, radius=16.5704, center=<25.1594 -1.1002 2.7871>, rotation=<rot 0 -0.6705 -80.462>}]
 hair_02 [{shapeType=Capsule, length=30.8089, radius=27.0256, center=<21.9073 0.3213 -0.9765>, rotation=<rot 21.8907 -90 16.2836>}]
 mouth_low_c [{shapeType=Capsule, length=30.6491, radius=11.8263, center=<-6.8843 -0.082 2.792>, rotation=<rot 0 0 -90>}]
}
constraints [
 {bone1Name=Ramen, bone2Name=Root, swing1LimitDegrees=45, swing2LimitDegrees=45, twistLimitDegrees=45}
 {bone1Name=cover_01, bone2Name=Ramen, swing1LimitDegrees=45, swing2LimitDegrees=45, twistLimitDegrees=45}
 {bone1Name=cover_03, bone2Name=cover_01, swing1LimitDegrees=45, swing2LimitDegrees=45, twistLimitDegrees=45}
 {bone1Name=cover_04, bone2Name=cover_03, swing1LimitDegrees=45, swing2LimitDegrees=45, twistLimitDegrees=45}
 {bone1Name=cover_05, bone2Name=cover_04, swing1LimitDegrees=45, swing2LimitDegrees=45, twistLimitDegrees=45}
 {bone1Name=cover_06, bone2Name=cover_05, swing1LimitDegrees=45, swing2LimitDegrees=45, twistLimitDegrees=45}
 {bone1Name=cover_07, bone2Name=cover_06, swing1LimitDegrees=45, swing2LimitDegrees=45, twistLimitDegrees=45}
 {bone1Name=cover_08, bone2Name=cover_07, swing1LimitDegrees=45, swing2LimitDegrees=45, twistLimitDegrees=45}
 {bone1Name=cover_09, bone2Name=cover_08, swing1LimitDegrees=45, swing2LimitDegrees=45, twistLimitDegrees=45}
 {bone1Name=Cover_12, bone2Name=cover_09, swing1LimitDegrees=45, swing2LimitDegrees=45, twistLimitDegrees=45}
 {bone1Name=bob, bone2Name=Root, swing1LimitDegrees=45, swing2LimitDegrees=45, twistLimitDegrees=45}
 {bone1Name=mouth_low_c, bone2Name=bob, swing1LimitDegrees=45, swing2LimitDegrees=45, twistLimitDegrees=45}
 {bone1Name=eyebrow_l, bone2Name=bob, swing1LimitDegrees=45, swing2LimitDegrees=45, twistLimitDegrees=45}
 {bone1Name=eyebrow_r, bone2Name=bob, swing1LimitDegrees=45, swing2LimitDegrees=45, twistLimitDegrees=45}
 {bone1Name=hair_01, bone2Name=bob, swing1LimitDegrees=45, swing2LimitDegrees=45, twistLimitDegrees=45}
 {bone1Name=hair_02, bone2Name=hair_01, swing1LimitDegrees=45, swing2LimitDegrees=45, twistLimitDegrees=45}
]
